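# Niagara system for the Skeleton Hunter character cosmetic, trailing sprite smoke, mesh wisps and cyan hair sprites from the body, head and hands.
# Emitters position particles on skeletal mesh bones, using partitions and spawn groups to choose bones and offsets.
# Constant-rate emitters supply ambient effects, while two spawn-per-unit emitters add movement trails.
# User bools for hidden head, hidden hands and skydiving drive particle visibility tags on all emitters except Smoke_Body.
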NiagaraSystem /BRCosmetics/Effects/Fort_Effects/Effects/Characters/Athena_Parts/SkeletonHunter/NS_SkelHunter_Smoke
dependencies {
 usedDataInterfaces [
  &NiagaraDataInterfaceArrayInt32
  &NiagaraDataInterfaceArrayFloat3
  &NiagaraDataInterfaceArrayFloat
  &NiagaraDataInterfaceVectorCurve
  &NiagaraDataInterfaceColorCurve
  &NiagaraDataInterfaceCurve
  &NiagaraDataInterfaceSpriteRendererInfo
  &NiagaraDataInterfaceVector2DCurve
  &NiagaraDataInterfaceSkeletalMesh
 ]
 usedDynamicInputs []
 usedModules [
  &SystemState
  &self:SystemSpawnScript.SystemScriptSource.SystemScriptGraph.NiagaraNodeAssignment_0.SetVariables_#1
 ]
}
emitters {
 Hair_Decal {
  emitterUpdateScript [
   &EmitterState()
   &SpawnRate(SpawnRate=10, Use Spawn Probability=false)
  ]
  particleSpawnScript [
   &PartitionParticles()
   &InitializeParticle(
    Color=<rgba 0 0.6785 0.8 1>
    Lifetime Max=1.3
    Lifetime Min=0.8
    Lifetime Mode=::NewEnumerator1
    Sprite Rotation Angle=&SelectFloatFromArray(Direct Array Index=$Particles.PartitionParticles.Partition, Float Selection Array=DataInterface() {propertyValues {FloatData [-0.18,0.18]}})
    Sprite Rotation Mode=::NewEnumerator1
    Sprite Size Mode=::NewEnumerator1
    Uniform Sprite Size Max=6.5
    Uniform Sprite Size Min=3.75
    UsePositionOffset=false
    Write Material Random=true
   )
   &CameraOffset(Camera Offset Amount=6)
   &AddVelocity(Velocity=&RandomRangeVector(
     Minimum=&SelectVectorFromArray(
      Direct Array Index=$Particles.PartitionParticles.Partition
      Vector Selection Array=DataInterface() {propertyValues {FloatData [<2 0 0>, <-2 0 0>]}}
     )
     Maximum=&SelectVectorFromArray(
      Direct Array Index=$Particles.PartitionParticles.Partition
      Vector Selection Array=DataInterface() {propertyValues {FloatData [<2 0 -1>, <-2 0 -1>]}}
     )
    ))
  ]
  particleUpdateScript [
   &ParticleState()
   &ScaleColor(
    Scale Alpha=&FloatFromCurve(FloatCurve=DataInterface() {propertyValues {Curve {
        defaultValue=1
        keys [
         {tangentMode=RCTM_User, time=0, value=0}
         <at 0.5 1>
         {arriveTangent=-2, interpMode=RCIM_Cubic, leaveTangent=-2, time=0.75, value=1}
         <at 1 0 RCIM_Cubic>
        ]
       }}})
    Scale RGB=&VectorFromFloat(Value=&FloatFromCurve(FloatCurve=DataInterface() {propertyValues {Curve {keys [<at 0 3>,<at 1 1>]}}}))
    ScaleA=true
    ScaleRGB=true
   )
   &SkeletalMeshLocation(
    Bone / Socket Index=7
    Bone Sampling Mode=::NewEnumerator4
    MaskSpawnGroup=false
    Position Sampling=::NewEnumerator1
    Sampled Position Offset=&RandomRangeVector(
     Minimum=&SelectVectorFromArray(
      Direct Array Index=$Particles.PartitionParticles.Partition
      Vector Selection Array=DataInterface() {propertyValues {FloatData [<-3 1 -7>, <-3 1 7>]}}
     )
     Maximum=&SelectVectorFromArray(
      Direct Array Index=$Particles.PartitionParticles.Partition
      Vector Selection Array=DataInterface() {propertyValues {FloatData [<5 1 -7>, <5 1 7>]}}
     )
    )
    Skeletal Mesh=$System.Skeletal Mesh
   )
   &SolveForcesAndVelocity(Clamp Velocity=false, Limit Acceleration=false)
   /SetVariables(Particles.VisibilityTag=&MakeCustomIntFromBool(Bool=$User.bisHeadHidden, True INT=1, False INT=0)) {isSetParameters=true}
  ]
  rendererClasses [&NiagaraSpriteRendererProperties]
  renderers [{rendererClass=&NiagaraSpriteRendererProperties}]
 }
 Long_Wisps {
  emitterUpdateScript [
   &EmitterState()
   &SpawnRate(SpawnRate=7, Use Spawn Probability=false)
   &SpawnRate(SpawnGroup=1, SpawnRate=2, Use Spawn Probability=false) {enabled=false, module=SpawnRate001}
   &SpawnRate(SpawnGroup=2, SpawnRate=4, Use Spawn Probability=false) {module=SpawnRate002}
   &SpawnRate(SpawnGroup=3, SpawnRate=3, Use Spawn Probability=false) {module=SpawnRate003}
  ]
  particleSpawnScript [
   &PartitionParticles(Number Of Partitions=&SelectIntFromArray(Direct Array Index=&SpawnGroup(), Int Selection Array=DataInterface() {propertyValues {IntData [2,1,2,1,1]}}))
   &InitializeParticle(
    Color Maximum=<rgba 1 1 1 60>
    Color Minimum=<rgba 0 0.1 0 60>
    Color Mode=::NewEnumerator2
    Lifetime Max=1.5
    Lifetime Mode=::NewEnumerator1
    Mesh Scale Max=&SelectVectorFromArray(
     Direct Array Index=&SpawnGroup()
     Vector Selection Array=DataInterface() {propertyValues {FloatData [<0.1 0.025 0.1>, <0.1 0.025 0.1>, <0.1 0.025 0.1>, <0.12 0.03 0.1>]}}
    )
    Mesh Scale Min=&SelectVectorFromArray(
     Direct Array Index=&SpawnGroup()
     Vector Selection Array=DataInterface() {propertyValues {FloatData [<0.05 0.01 0.1>, <0.05 0.008 0.1>, <0.05 0.008 0.1>, <0.06 0.02 0.1>]}}
    )
    Mesh Scale Mode=::NewEnumerator2
    UsePositionOffset=false
    Write Material Random=true
   )
   &AddVelocity(Velocity=&RandomRangeVector(Minimum=<0.1 1 0.5>, Maximum=<0.3 1 0.5>))
   &CameraOffset(Camera Offset Amount=&SelectFloatFromArray(
     Direct Array Index=&SpawnGroup()
     Float Selection Array=DataInterface() {propertyValues {FloatData [10,0,5,14,10]}}
    ))
  ]
  particleUpdateScript [
   &ParticleState()
   &ScaleMeshSize(Scale Factor=&VectorFromCurve(VectorCurve=DataInterface() {propertyValues {XCurve {keys [<at 0 0.1>,<at 1 10>]}, YCurve {keys [<at 0 1>,<at 1 8>]}, ZCurve {keys [<at 0 1>,<at 1 1>]}}}))
   &SkeletalMeshLocation(
    Bone / Socket Index=&MaskIntBySpawnGroup(
     Masked Int=&SelectIntFromArray(Direct Array Index=$Particles.PartitionParticles.Partition, Int Selection Array=DataInterface() {propertyValues {IntData [0,1]}})
     Passthrough Int=&MaskIntBySpawnGroup(
      Masked Int=3
      Passthrough Int=&MaskIntBySpawnGroup(
       Masked Int=&SelectIntFromArray(Direct Array Index=$Particles.PartitionParticles.Partition, Int Selection Array=DataInterface() {propertyValues {IntData [4,5]}})
       Passthrough Int=7
       Spawn Group Mask=2
      )
      Spawn Group Mask=1
     )
     Spawn Group Mask=0
    )
    Bone Sampling Mode=::NewEnumerator4
    MaskSpawnGroup=false
    Orientation Sampling=::NewEnumerator1
    Sampled Position Offset=&MaskVectorBySpawnGroup(
     Masked Vector=&SelectVectorFromArray(
      Direct Array Index=$Particles.PartitionParticles.Partition
      Vector Selection Array=DataInterface() {propertyValues {FloatData [<0 6 -5>, <0 6 -4>]}}
     )
     Passthrough Vector=&MaskVectorBySpawnGroup(
      Masked Vector=<0 4 -3>
      Passthrough Vector=&MaskVectorBySpawnGroup(Masked Vector=<0 0 0>, Passthrough Vector=<8 -6 0>, Spawn Group Mask=2)
      Spawn Group Mask=1
     )
    )
    Skeletal Mesh=$System.Skeletal Mesh
   )
   &SolveForcesAndVelocity(Clamp Velocity=false, Limit Acceleration=false)
   &ScaleColor(
    Scale Alpha=&FloatFromCurve(FloatCurve=DataInterface() {propertyValues {Curve {keys [
         <at 0 0 RCIM_Cubic>
         {arriveTangent=3, interpMode=RCIM_Cubic, leaveTangent=3, time=0.25, value=1}
         {arriveTangent=-0.6579, interpMode=RCIM_Cubic, leaveTangent=-0.6579, time=0.4, value=1.2}
         {arriveTangent=-2.5714, interpMode=RCIM_Cubic, leaveTangent=-2.5714, time=0.63, value=0.75}
         {arriveTangent=-2.2069, interpMode=RCIM_Cubic, leaveTangent=-2.2069, time=0.75, value=0.3}
         {arriveTangent=-1.2, interpMode=RCIM_Cubic, leaveTangent=-1.2, time=0.92, value=0.11}
         <at 1 0 RCIM_Cubic>
        ]}}})
    Scale RGB=&MakeVectorFromLinearColorRGB(Color=&ColorFromCurve(ColorCurve=DataInterface() {propertyValues {
        AlphaCurve {keys []}
        BlueCurve {keys [<at 0 1 RCIM_Cubic>,<at 0.25 1 RCIM_Cubic>]}
        GreenCurve {keys [<at 0 3 RCIM_Cubic>,<at 0.25 0.3 RCIM_Cubic>]}
        RedCurve {keys [<at 0 1 RCIM_Cubic>,<at 0.25 1 RCIM_Cubic>]}
       }}))
    ScaleA=true
    ScaleRGB=true
   )
   &DynamicMaterialParameters(
    Index 0 Param 1=&FloatFromCurve(FloatCurve=DataInterface() {propertyValues {Curve {keys [<at 0 5>,<at 1 0>]}}})
    Index 0 Param 2=&RandomRangeFloat(Minimum=30, Maximum=60)
    Index 0 Param 3=&FloatFromCurve(FloatCurve=DataInterface() {propertyValues {Curve {keys [<at 0 3>,<at 1 12>]}}})
    Index 0 Param 4=&FloatFromCurve(FloatCurve=DataInterface() {propertyValues {Curve {keys [
         <at 0 0.25 RCIM_Cubic>
         {arriveTangent=-0.4519, interpMode=RCIM_Cubic, leaveTangent=-0.4519, tangentMode=RCTM_User, time=0.82, value=0.22}
         {arriveTangent=-1.2222, interpMode=RCIM_Cubic, leaveTangent=-1.2222, time=0.9, value=0.1}
         <at 1 0 RCIM_Cubic>
        ]}}})
    Module.Index0Float0Write=true
    Module.Index0Float1Write=true
    Module.Index0Float2Write=true
    Module.Index0Float3Write=true
   )
   /SetVariables(Particles.VisibilityTag=&self:PartHIddenSpawnGroupCheck(
     is Hand Hidden:NiagaraBool=$User.bisHandHidden
     Selection:NiagaraInt32=&SpawnGroup()
     is Head HIdden:NiagaraBool=$User.bisHeadHidden
    )) {isSetParameters=true}
  ]
  rendererClasses [&NiagaraMeshRendererProperties]
  renderers [{rendererClass=&NiagaraMeshRendererProperties, rendererIndex=0}]
 }
 Long_Wisps_Body {
  emitterUpdateScript [
   &EmitterState()
   &SpawnRate(SpawnRate=3, Use Spawn Probability=false)
  ]
  particleSpawnScript [
   &PartitionParticles(Number Of Partitions=5)
   &InitializeParticle(
    Color Maximum=<rgba 1 0.2 1 60>
    Color Minimum=<rgba 0 0.1 0 60>
    Color Mode=::NewEnumerator2
    Lifetime Max=2.6
    Lifetime Min=1.6
    Lifetime Mode=::NewEnumerator1
    Mesh Scale Max=<0.3 0.06 0.1>
    Mesh Scale Min=<0.015 0.05 0.1>
    Mesh Scale Mode=::NewEnumerator2
    UsePositionOffset=false
    Write Material Random=true
   )
   &AddVelocity(Velocity=<-1 -0.5 0>)
   &CameraOffset(Camera Offset Amount=10)
  ]
  particleUpdateScript [
   &ParticleState()
   &SkeletalMeshLocation(
    Bone / Socket Index=&Add_Integer(A=$Particles.PartitionParticles.Partition, B=8)
    Bone Sampling Mode=::NewEnumerator4
    MaskSpawnGroup=false
    Orientation Sampling=::NewEnumerator1
    Position Sampling=::NewEnumerator1
    Sampled Position Offset=<10 7 -5>
    Skeletal Mesh=$System.Skeletal Mesh
   )
   &ScaleSpriteSize(
    Non-Uniform Curve Sprite Scale=DataInterface() {propertyValues {
      XCurve {keys [{tangentMode=RCTM_User, time=0, value=0.1}, {tangentMode=RCTM_User, time=1, value=10}]}
      YCurve {keys [{tangentMode=RCTM_User, time=0, value=1}, {tangentMode=RCTM_User, time=1, value=8}]}
     }}
    Scale Sprite Size Mode=::NewEnumerator3
   )
   &ScaleColor(
    Scale Alpha=&FloatFromCurve(FloatCurve=DataInterface() {propertyValues {Curve {keys [
         {arriveTangent=3.2785, interpMode=RCIM_Cubic, leaveTangent=3.2785, tangentMode=RCTM_User, time=0, value=0}
         {arriveTangent=2.7504, interpMode=RCIM_Cubic, leaveTangent=2.7504, tangentMode=RCTM_User, time=0.25, value=1}
         {interpMode=RCIM_Cubic, tangentMode=RCTM_User, time=0.4, value=1.2}
         {arriveTangent=-3.0964, interpMode=RCIM_Cubic, leaveTangent=-3.0964, tangentMode=RCTM_User, time=0.625, value=0.75}
         {arriveTangent=-2.5585, interpMode=RCIM_Cubic, leaveTangent=-2.5585, tangentMode=RCTM_User, time=0.75, value=0.275}
         {arriveTangent=-0.8132, interpMode=RCIM_Cubic, leaveTangent=-0.8132, tangentMode=RCTM_User, time=0.9, value=0.11}
         {arriveTangent=-1.1766, interpMode=RCIM_Cubic, leaveTangent=-1.1766, tangentMode=RCTM_User, time=1, value=0}
        ]}}})
    Scale RGB=&MakeVector(
     X=1
     Y=&FloatFromCurve(FloatCurve=DataInterface() {propertyValues {Curve {keys [<at 0 5>,<at 1 1>]}}})
     Z=1
    )
    ScaleA=true
    ScaleRGB=true
   )
   &DynamicMaterialParameters(
    Index 0 Param 1=&FloatFromCurve(FloatCurve=DataInterface() {propertyValues {Curve {keys [<at 0 5>,<at 1 0>]}}})
    Index 0 Param 2=&RandomRangeFloat(Minimum=30, Maximum=60)
    Index 0 Param 3=&FloatFromCurve(FloatCurve=DataInterface() {propertyValues {Curve {keys [<at 0 3>,<at 1 12>]}}})
    Index 0 Param 4=&FloatFromCurve(FloatCurve=DataInterface() {propertyValues {Curve {keys [
         <at 0 0.25 RCIM_Cubic>
         {arriveTangent=-0.4519, interpMode=RCIM_Cubic, leaveTangent=-0.4519, tangentMode=RCTM_User, time=0.82, value=0.22}
         {arriveTangent=-1.2222, interpMode=RCIM_Cubic, leaveTangent=-1.2222, time=0.9, value=0.1}
         <at 1 0 RCIM_Cubic>
        ]}}})
    Module.Index0Float0Write=true
    Module.Index0Float1Write=true
    Module.Index0Float2Write=true
    Module.Index0Float3Write=true
   )
   &SolveForcesAndVelocity(Clamp Velocity=false, Limit Acceleration=false)
   /SetVariables(Particles.VisibilityTag=&MakeCustomIntFromBool(Bool=$User.isSkydiving, True INT=1, False INT=0)) {isSetParameters=true, module=SetVariables_#2}
  ]
  rendererClasses [&NiagaraMeshRendererProperties]
  renderers [{rendererClass=&NiagaraMeshRendererProperties, rendererIndex=0}]
 }
 Long_Wisps_Body_Velocity {
  emitterUpdateScript [
   &EmitterState()
   &SpawnPerUnit(Module.Use Max Movement Threshold=true, Module.Use Movement Tolerance=false, Use Spawn Probability=false)
  ]
  particleSpawnScript [
   &PartitionParticles(Number Of Partitions=5)
   &InitializeParticle(
    Color Maximum=<rgba 1 0.2 1 60>
    Color Minimum=<rgba 0 0.1 0 60>
    Color Mode=::NewEnumerator2
    Lifetime Max=2.6
    Lifetime Min=1.6
    Lifetime Mode=::NewEnumerator1
    Mesh Scale Max=<0.3 0.06 0.1>
    Mesh Scale Min=<0.015 0.05 0.1>
    Mesh Scale Mode=::NewEnumerator2
    UsePositionOffset=false
    Write Material Random=true
   )
   &CameraOffset(Camera Offset Amount=10)
  ]
  particleUpdateScript [
   &ParticleState()
   &SkeletalMeshLocation(
    Bone / Socket Index=&Add_Integer(A=$Particles.PartitionParticles.Partition, B=8)
    Bone Sampling Mode=::NewEnumerator4
    MaskSpawnGroup=false
    Orientation Sampling=::NewEnumerator1
    Position Sampling=::NewEnumerator1
    Sampled Position Offset=<10 7 -5>
    Skeletal Mesh=$System.Skeletal Mesh
    Source Movement Scale:Vector3f=<0.1 0.1 0.1>
   )
   &ScaleSpriteSize(
    Non-Uniform Curve Sprite Scale=DataInterface() {propertyValues {
      XCurve {keys [{tangentMode=RCTM_User, time=0, value=0.1}, {tangentMode=RCTM_User, time=1, value=10}]}
      YCurve {keys [{tangentMode=RCTM_User, time=0, value=1}, {tangentMode=RCTM_User, time=1, value=8}]}
     }}
    Scale Sprite Size Mode=::NewEnumerator3
   )
   &ScaleColor(
    Scale Alpha=&FloatFromCurve(FloatCurve=DataInterface() {propertyValues {Curve {keys [
         {arriveTangent=3.2785, interpMode=RCIM_Cubic, leaveTangent=3.2785, tangentMode=RCTM_User, time=0, value=0}
         {arriveTangent=2.7504, interpMode=RCIM_Cubic, leaveTangent=2.7504, tangentMode=RCTM_User, time=0.25, value=1}
         {interpMode=RCIM_Cubic, tangentMode=RCTM_User, time=0.4, value=1.2}
         {arriveTangent=-3.0964, interpMode=RCIM_Cubic, leaveTangent=-3.0964, tangentMode=RCTM_User, time=0.625, value=0.75}
         {arriveTangent=-2.5585, interpMode=RCIM_Cubic, leaveTangent=-2.5585, tangentMode=RCTM_User, time=0.75, value=0.275}
         {arriveTangent=-0.8132, interpMode=RCIM_Cubic, leaveTangent=-0.8132, tangentMode=RCTM_User, time=0.9, value=0.11}
         {arriveTangent=-1.1766, interpMode=RCIM_Cubic, leaveTangent=-1.1766, tangentMode=RCTM_User, time=1, value=0}
        ]}}})
    Scale RGB=&MakeVector(
     X=1
     Y=&FloatFromCurve(FloatCurve=DataInterface() {propertyValues {Curve {keys [<at 0 5>,<at 1 1>]}}})
     Z=1
    )
    ScaleA=true
    ScaleRGB=true
   )
   &DynamicMaterialParameters(
    Index 0 Param 1=&FloatFromCurve(FloatCurve=DataInterface() {propertyValues {Curve {keys [<at 0 5>,<at 1 0>]}}})
    Index 0 Param 2=&RandomRangeFloat(Minimum=30, Maximum=60)
    Index 0 Param 3=&FloatFromCurve(FloatCurve=DataInterface() {propertyValues {Curve {keys [<at 0 3>,<at 1 12>]}}})
    Index 0 Param 4=&FloatFromCurve(FloatCurve=DataInterface() {propertyValues {Curve {keys [
         <at 0 0.25 RCIM_Cubic>
         {arriveTangent=-0.4519, interpMode=RCIM_Cubic, leaveTangent=-0.4519, tangentMode=RCTM_User, time=0.82, value=0.22}
         {arriveTangent=-1.2222, interpMode=RCIM_Cubic, leaveTangent=-1.2222, time=0.9, value=0.1}
         <at 1 0 RCIM_Cubic>
        ]}}})
    Module.Index0Float0Write=true
    Module.Index0Float1Write=true
    Module.Index0Float2Write=true
    Module.Index0Float3Write=true
   )
   &InheritVelocity(Inherited Velocity Amount Scale=<0.1 0.1 0.1>, Speed Limit=true, UseSpeedThreshold=false) {enabled=false}
   &Drag(Drag=2, Use Linear Drag=true, Use Rotational Drag=true)
   &SolveForcesAndVelocity(Clamp Velocity=false, Limit Acceleration=false)
   /SetVariables(Particles.VisibilityTag=&MakeCustomIntFromBool(Bool=$User.isSkydiving, True INT=1, False INT=0)) {isSetParameters=true, module=SetVariables_#2}
  ]
  rendererClasses [&NiagaraMeshRendererProperties]
  renderers [{rendererClass=&NiagaraMeshRendererProperties, rendererIndex=0}]
 }
 Smoke_Body {
  emitterUpdateScript [
   &EmitterState()
   &SpawnRate(SpawnRate=2, Use Spawn Probability=false)
  ]
  particleSpawnScript [
   &PartitionParticles()
   &InitializeParticle(
    Color Maximum=<rgba 0.05 0.5 0.6 1.3>
    Color Minimum=<rgba 0.01 0.1 0.2 0.8>
    Color Mode=::NewEnumerator2
    Lifetime Max=2.8
    Lifetime Min=1.5
    Lifetime Mode=::NewEnumerator1
    Sprite Rotation Mode=::NewEnumerator2
    Sprite Size Mode=::NewEnumerator1
    Uniform Sprite Size Max=50
    Uniform Sprite Size Min=30
    UsePositionOffset=false
    Write Material Random=true
   )
   &CameraOffset(Camera Offset Amount=&RandomRangeFloat(Minimum=8, Maximum=12))
   &AddVelocity(Velocity=&RandomRangeVector(Minimum=<3 -5 3>, Maximum=<8 5 5>))
   &DynamicMaterialParameters(
    Index 0 Param 1=&RandomRangeFloat(Minimum=0, Maximum=1)
    Index 0 Param 2=&RandomRangeFloat(Minimum=1, Maximum=2)
    Module.Index0Float0Write=true
    Module.Index0Float1Write=true
    Module.Index0Float2Write=true
    Module.Index0Float3Write=false
   )
  ]
  particleUpdateScript [
   &ParticleState()
   &ScaleSpriteSize(Uniform Curve Sprite Scale=DataInterface() {propertyValues {Curve {keys [
        {arriveTangent=5.8759, interpMode=RCIM_Cubic, leaveTangent=5.8759, tangentMode=RCTM_User, time=0, value=0.2}
        {
         arriveTangent=1.129
         arriveTangentWeight=0.0436
         interpMode=RCIM_Cubic
         leaveTangent=1.129
         leaveTangentWeight=0.3704
         tangentMode=RCTM_User
         time=0.1
         value=0.7
        }
        <at 0.95 1 RCIM_Cubic>
       ]}}})
   &SpriteRotationRate(Rotation Rate=&RandomRangeFloat(Minimum=54, Maximum=108))
   &SkeletalMeshLocation(
    Bone / Socket Index=&SelectIntFromArray(Direct Array Index=$Particles.PartitionParticles.Partition, Int Selection Array=DataInterface() {propertyValues {IntData [0,2]}})
    Bone Sampling Mode=::NewEnumerator4
    MaskSpawnGroup=false
    Position Sampling=::NewEnumerator1
    Sampled Position Offset=<9 15 0>
    Skeletal Mesh=$System.Skeletal Mesh
   )
   &SubUVAnimation(
    Direct Frame Index:NiagaraFloat=&FloatFromCurve(FloatCurve=DataInterface() {propertyValues {Curve {keys [{arriveTangent=43.5305, interpMode=RCIM_Cubic, leaveTangent=43.5305, tangentMode=RCTM_User, time=0, value=25}, <at 1 55 RCIM_Cubic>]}}})
    "Sprite Renderer "=DataInterface() {propertyValues {SpriteRenderer=&self:Smoke_Body_0.NiagaraSpriteRendererProperties_0}}
    SubUV Animation Mode:ENiagara_SubUVLookupModeV2=::NewEnumerator5
   )
   &ScaleColor(
    Scale Alpha=&FloatFromCurve(FloatCurve=DataInterface() {propertyValues {Curve {keys [
         {arriveTangent=2.6457, interpMode=RCIM_Cubic, leaveTangent=2.6457, tangentMode=RCTM_User, time=0, value=0}
         {arriveTangent=3.2466, interpMode=RCIM_Cubic, leaveTangent=3.2466, tangentMode=RCTM_User, time=0.25, value=1}
         {arriveTangent=-1.3333, interpMode=RCIM_Cubic, leaveTangent=-1.3333, time=0.8, value=1.8}
         <at 1 0 RCIM_Cubic>
        ]}}})
    Scale RGB=&VectorFromFloat(Value=&FloatFromCurve(FloatCurve=DataInterface() {propertyValues {Curve {keys [<at 0 1 RCIM_Cubic>,<at 0.5 0 RCIM_Cubic>]}}}))
    ScaleA=true
    ScaleRGB=true
   )
   &Drag(Use Linear Drag=true, Use Rotational Drag=true)
   &SolveForcesAndVelocity(Clamp Velocity=false, Limit Acceleration=false)
  ]
  rendererClasses [&NiagaraSpriteRendererProperties]
  renderers [{rendererClass=&NiagaraSpriteRendererProperties}]
 }
 Smoke_Hands {
  emitterUpdateScript [
   &EmitterState()
   &SpawnRate(SpawnRate=4, Use Spawn Probability=false)
  ]
  particleSpawnScript [
   &PartitionParticles()
   &InitializeParticle(
    Color Maximum=<rgba 0.05 0.5 0.6 1.3>
    Color Minimum=<rgba 0.01 0.1 0.2 0.8>
    Color Mode=::NewEnumerator2
    Lifetime Max=2.5
    Lifetime Min=1.5
    Lifetime Mode=::NewEnumerator1
    Sprite Rotation Mode=::NewEnumerator2
    Sprite Size Mode=::NewEnumerator1
    Uniform Sprite Size Max=42
    Uniform Sprite Size Min=25
    UsePositionOffset=false
    Write Material Random=true
   )
   &CameraOffset(Camera Offset Amount=&RandomRangeFloat(Minimum=5, Maximum=8))
   &AddVelocity(Velocity=&RandomRangeVector(Minimum=<6 -5 3>, Maximum=<16 5 5>))
   &DynamicMaterialParameters(
    Index 0 Param 1=&RandomRangeFloat(Minimum=0, Maximum=1)
    Index 0 Param 2=&RandomRangeFloat(Minimum=1, Maximum=2)
    Module.Index0Float0Write=true
    Module.Index0Float1Write=true
    Module.Index0Float2Write=true
    Module.Index0Float3Write=false
   )
  ]
  particleUpdateScript [
   &ParticleState()
   &ScaleSpriteSize(Uniform Curve Sprite Scale=DataInterface() {propertyValues {Curve {keys [
        {arriveTangent=5.8759, interpMode=RCIM_Cubic, leaveTangent=5.8759, tangentMode=RCTM_User, time=0, value=0.2}
        {
         arriveTangent=1.129
         arriveTangentWeight=0.0436
         interpMode=RCIM_Cubic
         leaveTangent=1.129
         leaveTangentWeight=0.3704
         tangentMode=RCTM_User
         time=0.1
         value=0.7
        }
        <at 0.95 1 RCIM_Cubic>
       ]}}})
   &SpriteRotationRate(Rotation Rate=&RandomRangeFloat(Minimum=54, Maximum=108))
   &SkeletalMeshLocation(
    Bone / Socket Index=&Add_Integer(A=$Particles.PartitionParticles.Partition, B=4)
    Bone Sampling Mode=::NewEnumerator4
    MaskSpawnGroup=false
    Position Sampling=::NewEnumerator1
    Sampled Position Offset=&SelectVectorFromArray(
     Direct Array Index=$Particles.PartitionParticles.Partition
     Vector Selection Array=DataInterface() {propertyValues {FloatData [<8 0 0>, <-8 0 0>]}}
    )
    Skeletal Mesh=$System.Skeletal Mesh
   )
   &SubUVAnimation(
    Direct Frame Index:NiagaraFloat=&FloatFromCurve(FloatCurve=DataInterface() {propertyValues {Curve {keys [{arriveTangent=43.5305, interpMode=RCIM_Cubic, leaveTangent=43.5305, tangentMode=RCTM_User, time=0, value=25}, <at 1 55 RCIM_Cubic>]}}})
    "Sprite Renderer "=DataInterface() {propertyValues {SpriteRenderer=&self:Smoke_Hands_0.NiagaraSpriteRendererProperties_0}}
    SubUV Animation Mode:ENiagara_SubUVLookupModeV2=::NewEnumerator5
   )
   &ScaleColor(
    Scale Alpha=&FloatFromCurve(FloatCurve=DataInterface() {propertyValues {Curve {keys [
         {arriveTangent=2.6457, interpMode=RCIM_Cubic, leaveTangent=2.6457, tangentMode=RCTM_User, time=0, value=0}
         {arriveTangent=3.2466, interpMode=RCIM_Cubic, leaveTangent=3.2466, tangentMode=RCTM_User, time=0.25, value=1}
         {arriveTangent=-1.3333, interpMode=RCIM_Cubic, leaveTangent=-1.3333, time=0.8, value=1.8}
         <at 1 0 RCIM_Cubic>
        ]}}})
    Scale RGB=&VectorFromFloat(Value=&FloatFromCurve(FloatCurve=DataInterface() {propertyValues {Curve {keys [<at 0 1 RCIM_Cubic>,<at 0.5 0.2 RCIM_Cubic>]}}}))
    ScaleA=true
    ScaleRGB=true
   )
   &Drag(Use Linear Drag=true, Use Rotational Drag=true)
   &SolveForcesAndVelocity(Clamp Velocity=false, Limit Acceleration=false)
   /SetVariables(Particles.VisibilityTag=&MakeCustomIntFromBool(Bool=$User.bisHandHidden, True INT=1, False INT=0)) {isSetParameters=true}
  ]
  rendererClasses [&NiagaraSpriteRendererProperties]
  renderers [{rendererClass=&NiagaraSpriteRendererProperties}]
 }
 Smoke_Loop_Velocity {
  emitterUpdateScript [
   &EmitterState()
   &SpawnPerUnit(Module.Use Max Movement Threshold=true, Module.Use Movement Tolerance=false, Use Spawn Probability=false)
  ]
  particleSpawnScript [
   &InitializeParticle(
    Color=&MakeLinearColorFromVectorAndFloat(
     "Vector (RGB)"=&MakeVectorFromLinearColorRGB(Color=<rgba 0 0.05 0.1 1>)
     "Float (Alpha)"=&RandomRangeFloat(Minimum=0.6, Maximum=0.9)
    )
    Lifetime Max=1.8
    Lifetime Min=1.2
    Lifetime Mode=::NewEnumerator1
    Sprite Rotation Mode=::NewEnumerator2
    Sprite Size Mode=::NewEnumerator1
    Uniform Sprite Size Max=130
    Uniform Sprite Size Min=90
    UsePositionOffset=false
    Write Material Random=true
   )
   &CameraOffset(Camera Offset Amount=&RandomRangeFloat(Minimum=15, Maximum=30))
   &AddVelocity(Velocity=&RandomRangeVector(Minimum=<-15 -15 0>, Maximum=<15 15 0>))
  ]
  particleUpdateScript [
   &ParticleState()
   &ScaleSpriteSize(Uniform Curve Sprite Scale=DataInterface() {propertyValues {Curve {keys [
        {arriveTangent=1.4318, interpMode=RCIM_Cubic, leaveTangent=1.4318, tangentMode=RCTM_User, time=0, value=0.6}
        {arriveTangent=1.4071, interpMode=RCIM_Cubic, leaveTangent=1.4071, tangentMode=RCTM_User, time=0.25, value=0.96}
        {arriveTangent=-1.0133, interpMode=RCIM_Cubic, leaveTangent=-1.0133, time=0.5, value=1.11}
        {arriveTangent=-3.1761, interpMode=RCIM_Cubic, leaveTangent=-3.1761, tangentMode=RCTM_User, time=1, value=0.2}
       ]}}})
   &ScaleColor(
    Scale Alpha=&FloatFromCurve(FloatCurve=DataInterface() {propertyValues {Curve {
        defaultValue=1
        keys [
         {arriveTangent=9.0725, interpMode=RCIM_Cubic, leaveTangent=9.0725, tangentMode=RCTM_User, time=0, value=0}
         {arriveTangent=-2.5249, interpMode=RCIM_Cubic, leaveTangent=-2.5249, tangentMode=RCTM_User, time=0.1408, value=0.9847}
         {arriveTangent=-0.0405, interpMode=RCIM_Cubic, leaveTangent=-0.0405, tangentMode=RCTM_User, time=1, value=0}
        ]
       }}})
    Scale RGB=&VectorFromFloat(Value=&FloatFromCurve(FloatCurve=DataInterface() {propertyValues {Curve {keys [<at 0 1>,<at 1 0.09>]}}}))
    ScaleA=true
    ScaleRGB=true
   )
   &SubUVAnimation(
    Direct Frame Index:NiagaraFloat=&FloatFromCurve(FloatCurve=DataInterface() {propertyValues {Curve {keys [{arriveTangent=23.4212, interpMode=RCIM_Cubic, leaveTangent=23.4212, tangentMode=RCTM_User, time=0, value=7}, <at 1 25 RCIM_Cubic>]}}})
    "Sprite Renderer "=DataInterface() {propertyValues {SpriteRenderer=&self:Smoke_Loop_Velocity_0.NiagaraSpriteRendererProperties_0}}
    SubUV Animation Mode:ENiagara_SubUVLookupModeV2=::NewEnumerator5
   )
   &DynamicMaterialParameters(
    Index 0 Param 1=&RandomRangeFloat(Minimum=0, Maximum=1)
    Index 0 Param 2=1
    Index 0 Param 3=&FloatFromCurve(FloatCurve=DataInterface() {propertyValues {Curve {keys [
         <at 0.5 1 RCIM_Cubic>
         {arriveTangent=-5.8278, interpMode=RCIM_Cubic, leaveTangent=-5.8278, tangentMode=RCTM_User, time=0.655, value=0.5}
         <at 0.8 0 RCIM_Cubic>
        ]}}})
    Module.Index0Float0Write=true
    Module.Index0Float1Write=true
    Module.Index0Float2Write=true
    Module.Index0Float3Write=false
   )
   &AccelerationForce(Acceleration=&RandomRangeVector(Minimum=<0 0 -30>, Maximum=<0 0 -80>))
   &SkeletalMeshLocation(
    Bone / Socket Index=6
    Bone Sampling Mode=::NewEnumerator4
    MaskSpawnGroup=false
    Position Sampling=::NewEnumerator1
    Sampled Position Offset=<-30 0 0>
    Skeletal Mesh=$System.Skeletal Mesh
    Source Movement Scale:Vector3f=<0.1 0.1 0.1>
   )
   &Drag(Drag=1.5, Use Linear Drag=true, Use Rotational Drag=true)
   &SolveForcesAndVelocity(Clamp Velocity=false, Limit Acceleration=false)
   /SetVariables(Particles.VisibilityTag=&MakeCustomIntFromBool(Bool=$User.isSkydiving, True INT=1, False INT=0)) {isSetParameters=true, module=SetVariables_#2}
  ]
  rendererClasses [&NiagaraSpriteRendererProperties]
  renderers [{rendererClass=&NiagaraSpriteRendererProperties}]
 }
}
summary {
 systemName=NS_SkelHunter_Smoke
 emitters [
  {emitterName=Long_Wisps, rendererClasses [&NiagaraMeshRendererProperties]}
  {emitterName=Smoke_Hands, rendererClasses [&NiagaraSpriteRendererProperties]}
  {emitterName=Smoke_Body, rendererClasses [&NiagaraSpriteRendererProperties]}
  {emitterName=Hair_Decal, rendererClasses [&NiagaraSpriteRendererProperties]}
  {emitterName=Smoke_Loop_Velocity, rendererClasses [&NiagaraSpriteRendererProperties]}
  {emitterName=Long_Wisps_Body, rendererClasses [&NiagaraMeshRendererProperties]}
  {emitterName=Long_Wisps_Body_Velocity, rendererClasses [&NiagaraMeshRendererProperties]}
 ]
 userVariables [
  User.bisHandHidden {type=NiagaraBool}
  User.bisHeadHidden {type=NiagaraBool}
  User.isSkydiving {type=NiagaraBool}
 ]
}
